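# Niagara system for the Dark Pit pickaxe's idle effect, layering cyan fire, smoke, goop and dark ribbon trails on the weapon.
# Emitters sample bones or sockets on user-supplied skeletal meshes; distance-spawned ribbons use a scratch-pad module to lock their heads.
# A system Timeline crossfades ribbons in and some head smoke out, with user running parameters driving ribbon lifetime and opacity.
# Fire and smoke use paired facing variants and time-of-day alpha scaling.
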NiagaraSystem /BRCosmetics/Weapons/FORT_Melee/Pickaxe_Dark_Pit_Male/FX/NS_Dark_Pit_Pick_IDle
dependencies {
 usedDataInterfaces [
  &NiagaraDataInterfaceArrayInt32
  &NiagaraDataInterfaceArrayFloat3
  &NiagaraDataInterfaceArrayFloat
  &NiagaraDataInterfaceCurve
  &NiagaraDataInterfaceColorCurve
 ]
 usedDynamicInputs [&VectorLength]
 usedModules [&SystemState, &Timeline]
}
emitters {
 Attached_Ribbon {
  emitterUpdateScript [
   &EmitterState(Inactive Response=::NewEnumerator1, Life Cycle Mode=::NewEnumerator1, Loop Duration=1, UseLoopDelay=false)
   /Game/Effects/Niagara/Modules/Spawn/FN_SpawnByEngineDistance(Multiplier=2, Step=1)
  ]
  particleSpawnScript [
   &InitializeRibbon(
    Color=<rgba 0.005 0.005 0.005 5>
    Lifetime=&Multiply_Float(A=0.075, B=$User.RunningLifetime)
    Module.Write Color=true
    Module.Write Lifetime=true
    Module.Write Mass=true
    Module.Write Material Random=true
    Module.Write Position=false
    Module.Write RibbonTwist=false
    Module.Write RibbonWidth=true
    Ribbon Width=8
   )
   /SetVariables(Particles.Stream:NiagaraInt32=&SpawnGroup()) {isSetParameters=true, module=SetVariables_#1}
   &FN_SetSpawnGroup(Number of Spawn Groups:NiagaraInt32=$Emitter.Spawn Multiplier)
   &VectorNoiseForce(Force Amount=10, Module.Use Random Noise Time Interval=false)
   &ApplyInitialForces(Apply Force to Position=true, Apply Rotational Force to Rotational Velocity=false)
  ]
  particleUpdateScript [
   &ParticleState()
   &ScaleColor(
    Scale Alpha=&Multiply_Float(
     A=&FloatFromCurve(
      FloatCurve=DataInterface() {propertyValues {Curve {keys [{interpMode=RCIM_Cubic, tangentMode=RCTM_User, time=0.6, value=1}, {interpMode=RCIM_Cubic, tangentMode=RCTM_User, time=1, value=0}]}}}
      Scale Curve=&Multiply_Float(A=0.05, B=$User.RunningOpacity)
     )
     B=$System.Timeline.NormalizedTime
    )
    ScaleA=true
    ScaleRGB=true
   ) {module=ScaleColor001}
   &SkeletalMeshLocation(
    MaskSpawnGroup=false
    Mesh Sampling Type=::NewEnumerator5
    Orientation Sampling=::NewEnumerator1
    Position Sampling=::NewEnumerator2
    Skeletal Mesh=$User.SkeletalMesh
    Socket Sampling Mode=::NewEnumerator5
   )
   &SolveForcesAndVelocity(Clamp Velocity=false, Limit Acceleration=false)
   &self:Attached_Ribbon_8.NiagaraScratchPadContainer_1.FN_Lock_Ribbon_Head(
    Emitter Count:NiagaraInt32=$Emitter.Count
    Locked Position:NiagaraPosition=$Particles.SkeletalMeshLocation.SampledPosition
    Particle Stream:NiagaraInt32=$Particles.Stream
   )
   &ScaleRibbonWidth(Ribbon Width Scale=&FloatFromCurve(
     FloatCurve=DataInterface() {propertyValues {Curve {keys [
         {arriveTangent=10, interpMode=RCIM_Cubic, leaveTangent=10, time=0, value=0.25}
         {arriveTangent=-0.0085, interpMode=RCIM_Cubic, leaveTangent=-0.0085, tangentMode=RCTM_User, time=0.1, value=1}
         {arriveTangent=-0.0405, interpMode=RCIM_Cubic, leaveTangent=-0.0405, tangentMode=RCTM_User, time=0.7009, value=1}
         {arriveTangent=-8.0089, interpMode=RCIM_Cubic, leaveTangent=-8.0089, tangentMode=RCTM_User, time=1, value=0}
        ]}}}
     Scale Curve=&Multiply_Float(A=0.8, B=$System.Timeline.NormalizedTime)
    ))
  ]
  rendererClasses [&NiagaraRibbonRendererProperties]
  renderers [{rendererClass=&NiagaraRibbonRendererProperties}]
 }
 Attached_Ribbon001 {
  emitterUpdateScript [
   &EmitterState(Inactive Response=::NewEnumerator1, Life Cycle Mode=::NewEnumerator1, Loop Duration=1, UseLoopDelay=false)
   &FN_SpawnByEngineDistance(Multiplier=2, Step=1)
  ]
  particleSpawnScript [
   &InitializeRibbon(
    Color=<rgba 0.005 0.005 0.005 5>
    Lifetime=&Multiply_Float(A=0.075, B=$User.RunningLifetime)
    Module.Write Color=true
    Module.Write Lifetime=true
    Module.Write Mass=true
    Module.Write Material Random=true
    Module.Write Position=false
    Module.Write RibbonTwist=false
    Module.Write RibbonWidth=true
    Ribbon Width=8
   )
   /SetVariables(Particles.Stream:NiagaraInt32=&SpawnGroup()) {isSetParameters=true, module=SetVariables_#1}
   &FN_SetSpawnGroup(Number of Spawn Groups:NiagaraInt32=$Emitter.Spawn Multiplier)
   &VectorNoiseForce(Force Amount=10, Module.Use Random Noise Time Interval=false)
   &ApplyInitialForces(Apply Force to Position=true, Apply Rotational Force to Rotational Velocity=false)
  ]
  particleUpdateScript [
   &ParticleState()
   &ScaleColor(
    Scale Alpha=&Multiply_Float(
     A=&FloatFromCurve(
      FloatCurve=DataInterface() {propertyValues {Curve {keys [{interpMode=RCIM_Cubic, tangentMode=RCTM_User, time=0.6, value=1}, {interpMode=RCIM_Cubic, tangentMode=RCTM_User, time=1, value=0}]}}}
      Scale Curve=&Multiply_Float(A=0.05, B=$User.RunningOpacity)
     )
     B=$System.Timeline.NormalizedTime
    )
    ScaleA=true
    ScaleRGB=true
   ) {module=ScaleColor001}
   &SkeletalMeshLocation(
    Bone / Socket Index=4
    MaskSpawnGroup=false
    Mesh Sampling Type=::NewEnumerator5
    Orientation Sampling=::NewEnumerator1
    Position Sampling=::NewEnumerator2
    Skeletal Mesh=$User.SkeletalMesh
    Socket Sampling Mode=::NewEnumerator5
   )
   &SolveForcesAndVelocity(Clamp Velocity=false, Limit Acceleration=false)
   &self:Attached_Ribbon001_10.NiagaraScratchPadContainer_1.FN_Lock_Ribbon_Head(
    Emitter Count:NiagaraInt32=$Emitter.Count
    Locked Position:NiagaraPosition=$Particles.SkeletalMeshLocation.SampledPosition
    Particle Stream:NiagaraInt32=$Particles.Stream
   )
   &ScaleRibbonWidth(Ribbon Width Scale=&FloatFromCurve(
     FloatCurve=DataInterface() {propertyValues {Curve {keys [
         {arriveTangent=10, interpMode=RCIM_Cubic, leaveTangent=10, time=0, value=0.25}
         {arriveTangent=-0.0085, interpMode=RCIM_Cubic, leaveTangent=-0.0085, tangentMode=RCTM_User, time=0.1, value=1}
         {arriveTangent=-0.0405, interpMode=RCIM_Cubic, leaveTangent=-0.0405, tangentMode=RCTM_User, time=0.7009, value=1}
         {arriveTangent=-8.0089, interpMode=RCIM_Cubic, leaveTangent=-8.0089, tangentMode=RCTM_User, time=1, value=0}
        ]}}}
     Scale Curve=0.8
    ))
  ]
  rendererClasses [&NiagaraRibbonRendererProperties]
  renderers [{rendererClass=&NiagaraRibbonRendererProperties}]
 }
 Attached_Ribbon002 {
  emitterUpdateScript [
   &EmitterState(Inactive Response=::NewEnumerator1, Life Cycle Mode=::NewEnumerator1, Loop Duration=1, UseLoopDelay=false)
   &FN_SpawnByEngineDistance(Multiplier=2, Step=1)
  ]
  particleSpawnScript [
   &InitializeRibbon(
    Color=<rgba 0.005 0.005 0.005 5>
    Lifetime=&Multiply_Float(A=0.075, B=$User.RunningLifetime)
    Module.Write Color=true
    Module.Write Lifetime=true
    Module.Write Mass=true
    Module.Write Material Random=true
    Module.Write Position=false
    Module.Write RibbonTwist=false
    Module.Write RibbonWidth=true
    Ribbon Width=8
   )
   /SetVariables(Particles.Stream:NiagaraInt32=&SpawnGroup()) {isSetParameters=true, module=SetVariables_#1}
   &FN_SetSpawnGroup(Number of Spawn Groups:NiagaraInt32=$Emitter.Spawn Multiplier)
   &VectorNoiseForce(Force Amount=10, Module.Use Random Noise Time Interval=false)
   &ApplyInitialForces(Apply Force to Position=true, Apply Rotational Force to Rotational Velocity=false)
  ]
  particleUpdateScript [
   &ParticleState()
   &ScaleColor(
    Scale Alpha=&Multiply_Float(
     A=&FloatFromCurve(
      FloatCurve=DataInterface() {propertyValues {Curve {keys [{interpMode=RCIM_Cubic, tangentMode=RCTM_User, time=0.6, value=1}, {interpMode=RCIM_Cubic, tangentMode=RCTM_User, time=1, value=0}]}}}
      Scale Curve=&Multiply_Float(A=0.05, B=$User.RunningOpacity)
     )
     B=$System.Timeline.NormalizedTime
    )
    ScaleA=true
    ScaleRGB=true
   ) {module=ScaleColor001}
   &SkeletalMeshLocation(
    Bone / Socket Index=9
    MaskSpawnGroup=false
    Mesh Sampling Type=::NewEnumerator5
    Orientation Sampling=::NewEnumerator1
    Position Sampling=::NewEnumerator2
    Skeletal Mesh=$User.SkeletalMesh
    Socket Sampling Mode=::NewEnumerator5
   )
   &SolveForcesAndVelocity(Clamp Velocity=false, Limit Acceleration=false)
   &self:Attached_Ribbon002_12.NiagaraScratchPadContainer_1.FN_Lock_Ribbon_Head(
    Emitter Count:NiagaraInt32=$Emitter.Count
    Locked Position:NiagaraPosition=$Particles.SkeletalMeshLocation.SampledPosition
    Particle Stream:NiagaraInt32=$Particles.Stream
   )
   &ScaleRibbonWidth(Ribbon Width Scale=&FloatFromCurve(
     FloatCurve=DataInterface() {propertyValues {Curve {keys [
         {arriveTangent=10, interpMode=RCIM_Cubic, leaveTangent=10, time=0, value=0.25}
         {arriveTangent=-0.0085, interpMode=RCIM_Cubic, leaveTangent=-0.0085, tangentMode=RCTM_User, time=0.1, value=1}
         {arriveTangent=-0.0405, interpMode=RCIM_Cubic, leaveTangent=-0.0405, tangentMode=RCTM_User, time=0.7009, value=1}
         {arriveTangent=-8.0089, interpMode=RCIM_Cubic, leaveTangent=-8.0089, tangentMode=RCTM_User, time=1, value=0}
        ]}}}
     Scale Curve=0.8
    ))
  ]
  rendererClasses [&NiagaraRibbonRendererProperties]
  renderers [{rendererClass=&NiagaraRibbonRendererProperties}]
 }
 Attached_Ribbon003 {
  emitterUpdateScript [
   &EmitterState(Inactive Response=::NewEnumerator1, Life Cycle Mode=::NewEnumerator1, Loop Duration=1, UseLoopDelay=false)
   &FN_SpawnByEngineDistance(Multiplier=2, Step=1)
  ]
  particleSpawnScript [
   &InitializeRibbon(
    Color=<rgba 0.005 0.005 0.005 5>
    Lifetime=&Multiply_Float(A=0.075, B=$User.RunningLifetime)
    Module.Write Color=true
    Module.Write Lifetime=true
    Module.Write Mass=true
    Module.Write Material Random=true
    Module.Write Position=false
    Module.Write RibbonTwist=false
    Module.Write RibbonWidth=true
    Ribbon Width=8
   )
   /SetVariables(Particles.Stream:NiagaraInt32=&SpawnGroup()) {isSetParameters=true, module=SetVariables_#1}
   &FN_SetSpawnGroup(Number of Spawn Groups:NiagaraInt32=$Emitter.Spawn Multiplier)
   &VectorNoiseForce(Force Amount=10, Module.Use Random Noise Time Interval=false)
   &ApplyInitialForces(Apply Force to Position=true, Apply Rotational Force to Rotational Velocity=false)
  ]
  particleUpdateScript [
   &ParticleState()
   &ScaleColor(
    Scale Alpha=&Multiply_Float(
     A=&FloatFromCurve(
      FloatCurve=DataInterface() {propertyValues {Curve {keys [{interpMode=RCIM_Cubic, tangentMode=RCTM_User, time=0.6, value=1}, {interpMode=RCIM_Cubic, tangentMode=RCTM_User, time=1, value=0}]}}}
      Scale Curve=&Multiply_Float(A=0.05, B=$User.RunningOpacity)
     )
     B=$System.Timeline.NormalizedTime
    )
    ScaleA=true
    ScaleRGB=true
   ) {module=ScaleColor001}
   &SkeletalMeshLocation(
    Bone / Socket Index=10
    MaskSpawnGroup=false
    Mesh Sampling Type=::NewEnumerator5
    Orientation Sampling=::NewEnumerator1
    Position Sampling=::NewEnumerator2
    Skeletal Mesh=$User.SkeletalMesh
    Socket Sampling Mode=::NewEnumerator5
   )
   &SolveForcesAndVelocity(Clamp Velocity=false, Limit Acceleration=false)
   &self:Attached_Ribbon003_14.NiagaraScratchPadContainer_1.FN_Lock_Ribbon_Head(
    Emitter Count:NiagaraInt32=$Emitter.Count
    Locked Position:NiagaraPosition=$Particles.SkeletalMeshLocation.SampledPosition
    Particle Stream:NiagaraInt32=$Particles.Stream
   )
   &ScaleRibbonWidth(Ribbon Width Scale=&FloatFromCurve(
     FloatCurve=DataInterface() {propertyValues {Curve {keys [
         {arriveTangent=10, interpMode=RCIM_Cubic, leaveTangent=10, time=0, value=0.25}
         {arriveTangent=-0.0085, interpMode=RCIM_Cubic, leaveTangent=-0.0085, tangentMode=RCTM_User, time=0.1, value=1}
         {arriveTangent=-0.0405, interpMode=RCIM_Cubic, leaveTangent=-0.0405, tangentMode=RCTM_User, time=0.7009, value=1}
         {arriveTangent=-8.0089, interpMode=RCIM_Cubic, leaveTangent=-8.0089, tangentMode=RCTM_User, time=1, value=0}
        ]}}}
     Scale Curve=0.8
    ))
  ]
  rendererClasses [&NiagaraRibbonRendererProperties]
  renderers [{rendererClass=&NiagaraRibbonRendererProperties}]
 }
 Goop {
  emitterUpdateScript [
   &EmitterState(Inactive Response=::NewEnumerator1, Life Cycle Mode=::NewEnumerator1, Loop Duration=2, UseLoopDelay=false)
   &SpawnRate(SpawnRate=3.5, Use Spawn Probability=false)
  ]
  particleSpawnScript [
   &InitializeParticle(
    Color=<rgba 0 0.8911 1 1>
    Lifetime Max=2.8
    Lifetime Min=2.5
    Lifetime Mode=::NewEnumerator1
    Mesh Scale Max=<0.6 0.6 1>
    Mesh Scale Min=<0.2 0.2 0.6>
    Mesh Scale Mode=::NewEnumerator2
    Position Offset=<0 0 -4>
    UsePositionOffset=true
    Write Material Random=true
   )
   &FN_SetSpawnGroup(Number of Spawn Groups=2)
   &DynamicMaterialParameters(
    Index 0 Param 1=2
    Index 0 Param 2=10
    Index 0 Param 3=3
    Index 0 Param 4=&RandomRangeFloat(Minimum=50, Maximum=55, Evaluation Type=::NewEnumerator1)
    Module.Param0WriteEnabled=true
    Module.Param1WriteEnabled=true
    Module.Param2WriteEnabled=true
    Module.Param3WriteEnabled=false
   ) {module=DynamicMaterialParameters002}
   &SkeletalMeshLocation(
    Bone / Socket Index=$Particles.SpawnGroup
    MaskSpawnGroup=false
    Mesh Sampling Type=::NewEnumerator5
    Skeletal Mesh=$User.Goop_SM
    Socket Sampling Mode=::NewEnumerator5
   )
   &SphereLocation(
    MaskSpawnGroup=false
    Override Seed=false
    Sphere Origin:NiagaraPosition=$Output.SkeletalMeshLocation.SampledPositionWithoutOffset
    Sphere Radius=5
    Surface Only=false
    UseNonUniformScale=false
    UseOffset=false
    UseSphereOrientation=false
   )
   &CameraOffset(Camera Offset Amount=&RandomRangeFloat(Minimum=5, Maximum=10))
  ]
  particleUpdateScript [
   &ParticleState()
   &DynamicMaterialParameters(
    Index 0 Param 3=&FloatFromCurve(
     FloatCurve=DataInterface() {propertyValues {Curve {keys [
         <at 0 0 RCIM_Cubic>
         {arriveTangent=2.4962, interpMode=RCIM_Cubic, leaveTangent=2.4962, tangentMode=RCTM_User, time=0.3, value=0.45}
         <at 1 1>
        ]}}}
     Scale Curve=2
    )
    Module.Param0WriteEnabled=false
    Module.Param1WriteEnabled=false
    Module.Param2WriteEnabled=true
    Module.Param3WriteEnabled=false
   )
   &ScaleMeshSize(Scale Factor=&VectorFromFloat(Value=&FloatFromCurve(
      FloatCurve=DataInterface() {propertyValues {Curve {keys [
          {arriveTangent=-0.026, interpMode=RCIM_Cubic, leaveTangent=-0.026, tangentMode=RCTM_User, time=0, value=0}
          {arriveTangent=0.1004, interpMode=RCIM_Cubic, leaveTangent=0.1004, tangentMode=RCTM_User, time=0.6, value=0.7}
          {arriveTangent=-3.4821, interpMode=RCIM_Cubic, leaveTangent=-3.4821, tangentMode=RCTM_User, time=1, value=0}
         ]}}}
      Scale Curve=0.6
     ))) {module=ScaleMeshSize001}
   &GravityForce(
    Coordinate Space=::Simulation
    Gravity=&RotateVectorByQuaternion(Vector To Rotate=<20 0 0>, Quaternion=$Particles.Initial.SkeletalMeshLocation.SampledRotation)
   ) {module=GravityForce001}
   &Drag(Use Linear Drag=true, Use Rotational Drag=false)
   &MeshRotationForce(
    Coordinate Space:ENiagaraMeshTransforms=::NewEnumerator3
    Rotation=&RandomRangeVector(Minimum=<-4 -4 -4>, Maximum=<4 4 4>)
   )
   &SolveForcesAndVelocity(Clamp Velocity=false, Limit Acceleration=false)
   &SolveRotationalForcesAndVelocity()
   &KillParticlesInVolume(Invert Volume=true, Sphere Radius=1500)
  ]
  rendererClasses [&NiagaraMeshRendererProperties]
  renderers [{rendererClass=&NiagaraMeshRendererProperties, rendererIndex=0}]
 }
 Head_Blue_Smoke {
  emitterUpdateScript [
   &EmitterState(Inactive Response=::NewEnumerator1, Life Cycle Mode=::NewEnumerator1, Loop Duration=2, UseLoopDelay=false)
   &SpawnRate(SpawnRate=1.5, Use Spawn Probability=false)
  ]
  particleSpawnScript [
   &InitializeParticle(
    Color=&RandomRangeLinearColor(Minimum=<rgba 0 0.6927 0.5762 1>, Maximum=<rgba 0.4582 1 0.9901 1>)
    Lifetime=&RandomRangeFloat(Minimum=1.2, Maximum=1.8)
    Mesh Scale=<0.5 1 1>
    Mesh Scale Mode=::NewEnumerator4
    Sprite Rotation Angle=0.5
    Sprite Rotation Mode=::NewEnumerator1
    Sprite Size=&Multiply_Vector2D_ByFloat(Float=1, Vector2D=<20 80>)
    Sprite Size Mode=::NewEnumerator4
    UsePositionOffset=false
    Write Material Random=true
   )
   &InitialMeshOrientation(Mesh Orientation Mode=::NewEnumerator2, Rotation=<-0.05 0 0>, Use Rotation Vector=true)
   &FN_SetSpawnGroup(Number of Spawn Groups=9)
  ]
  particleUpdateScript [
   &ParticleState(Kill Particles When Lifetime Has Elapsed=false, Let Infinitely Lived Particles Die When Emitter Deactivates=true)
   &SkeletalMeshLocation(
    MaskSpawnGroup=false
    Mesh Sampling Type=::NewEnumerator5
    Orientation Sampling=::NewEnumerator1
    Sampled Position Offset=&Add_Vector(A=<0 3 10>, B=&RandomRangeVector(Minimum=<-2 -6 -12>, Maximum=<10 -2 -5>))
    Skeletal Mesh=$User.AxeHead_SM
    Socket Sampling Mode=::NewEnumerator5
    Use Sampled Mesh Attribute as Ribbon ID=false
   ) {module=SkeletalMeshLocation001}
   &ScaleColor(
    Scale Alpha=&FloatFromCurve(FloatCurve=DataInterface() {propertyValues {Curve {keys [
         {interpMode=RCIM_Cubic, tangentMode=RCTM_User, time=0, value=0}
         {interpMode=RCIM_Cubic, tangentMode=RCTM_User, time=0.5, value=1}
         {tangentMode=RCTM_User, time=1, value=0}
        ]}}})
    ScaleA=true
    ScaleRGB=false
   )
   &SolveForcesAndVelocity(Clamp Velocity=false, Limit Acceleration=false)
   &CameraOffset(Camera Offset Amount=&RandomRangeFloat(Minimum=-6, Maximum=6))
   &SpriteFacingAndAlignment(
    Sprite Alignment=&RotateVectorByQuaternion(Vector To Rotate=<-1 0 0>, Quaternion=$Particles.SkeletalMeshLocation001.SampledRotation)
    Sprite Facing=&RotateVectorByQuaternion(Vector To Rotate=<0 1 0>, Quaternion=$Particles.SkeletalMeshLocation001.SampledRotation)
    WriteAlignment=true
    WriteFacing=true
   )
   &DynamicMaterialParameters(
    Index 0 Param 4=&RandomRangeFloat(Minimum=0, Maximum=1)
    Module.Param0WriteEnabled=true
    Module.Param0WriteEnabled1=true
    Module.Param0WriteEnabled2=true
    Module.Param0WriteEnabled3=true
    Module.Param1WriteEnabled=true
    Module.Param1WriteEnabled1=true
    Module.Param1WriteEnabled2=true
    Module.Param1WriteEnabled3=true
    Module.Param2WriteEnabled=true
    Module.Param2WriteEnabled1=true
    Module.Param2WriteEnabled2=true
    Module.Param2WriteEnabled3=true
    Module.Param3WriteEnabled=true
    Module.Param3WriteEnabled1=true
    Module.Param3WriteEnabled2=true
    Module.Param3WriteEnabled3=true
    Write Parameter Index 1=true
    Write Parameter Index 2=true
    Write Parameter Index 3=true
   )
  ]
  rendererClasses [&NiagaraSpriteRendererProperties]
  renderers [{rendererClass=&NiagaraSpriteRendererProperties}]
 }
 Head_Fire_FaceX {
  emitterUpdateScript [
   &EmitterState(Inactive Response=::NewEnumerator1, Life Cycle Mode=::NewEnumerator1, UseLoopDelay=false)
   &SpawnRate(SpawnRate=5, Use Spawn Probability=false)
  ]
  particleSpawnScript [
   &FN_SetSpawnGroup(Number of Spawn Groups=2)
   &InitializeParticle(
    Color=<rgba 0 0.7155 1 0.25>
    Lifetime Max=3.5
    Lifetime Min=2.5
    Lifetime Mode=::NewEnumerator1
    Material Random:NiagaraFloat=&RandomRangeFloat(Minimum=0, Maximum=100)
    Mesh Scale=&Multiply_Vector(
     A=&RandomRangeVector(Minimum=<0.75 0.75 0.25>, Maximum=<1 0.75 0.25>)
     B=&VectorFromFloat(Value=0.2)
    )
    Mesh Scale Mode=::NewEnumerator4
    Sprite Rotation Angle=0.5
    Sprite Rotation Mode=::NewEnumerator1
    Sprite Size Mode=::NewEnumerator3
    Uniform Sprite Size=&Multiply_Float(A=25, B=&RandomRangeFloat(Minimum=0.75, Maximum=1))
    UsePositionOffset=false
    Write Material Random=true
   ) {module=InitializeParticle001}
   &DynamicMaterialParameters(
    Index 0 Param 1=&RandomRangeFloat(Minimum=0, Maximum=0.6)
    Index 0 Param 2=1
    Index 0 Param 3=30
    Module.Param0WriteEnabled=true
    Module.Param1WriteEnabled=true
    Module.Param2WriteEnabled=true
    Module.Param3WriteEnabled=true
   )
   &AddVelocity(Velocity=<0 0 1>)
   &CameraOffset(Camera Offset Amount=3)
   /SetVariables(Particles.Scale=&Multiply_Vector(
     A=$Particles.Scale
     B=&MakeCustomVectorFromBool(Bool=&RandomBool(), True Vector=<1 1 1>, False Vector=<-1 1 1>)
    )) {isSetParameters=true, module=SetVariables_#2}
  ]
  particleUpdateScript [
   &ParticleState()
   &InterpolateOverTime(Rate Of Change=10, Target Vector Value:NiagaraPosition=$Particles.Position, Variable Type=::NewEnumerator2)
   &SkeletalMeshLocation(
    Bone / Socket Index=&SelectIntFromArray(Direct Array Index=$Particles.SpawnGroup, Int Selection Array=DataInterface() {propertyValues {IntData [0,1]}})
    MaskSpawnGroup=false
    Mesh Sampling Type=::NewEnumerator5
    Orientation Sampling=::NewEnumerator1
    Sampled Position Offset=&Add_Vector(
     A=&Add_Vector(A=<-5 3 10>, B=&RandomRangeVector(Minimum=<-2 -6 -12>, Maximum=<10 -2 -5>))
     B=&MakeVector(
      X=&FloatFromCurve(FloatCurve=DataInterface() {propertyValues {Curve {keys [<at 0 0>,<at 1 1>]}}}, Scale Curve=20)
      Y=0
      Z=0
     )
    )
    Skeletal Mesh=$User.Fire_SM
    Socket Sampling Mode=::NewEnumerator5
    Use Sampled Mesh Attribute as Ribbon ID=false
   ) {module=SkeletalMeshLocation001}
   /SetVariables(Particles.Interpolate Dynamic=&MakeFloat4FromVectorAndFloat("Vector (RGB)"=$Output.InterpolateOverTime.Moving Average, "Float (Alpha)"=0.5)) {isSetParameters=true, module=SetVariables_#3}
   &Drag(Drag=0.5, Use Linear Drag=true, Use Rotational Drag=false)
   &SolveForcesAndVelocity(Clamp Velocity=false, Limit Acceleration=false)
   &ScaleColor(
    Scale Alpha=&Multiply_Float(
     A=&FNTimeOfDayScalar(Night Scale=0.5, Morning Scale=0.7, Evening Scale=0.7)
     B=&FloatFromCurve(FloatCurve=DataInterface() {propertyValues {Curve {keys [
          {
           arriveTangentWeight=0.3952
           interpMode=RCIM_Cubic
           leaveTangentWeight=0.3952
           tangentMode=RCTM_User
           tangentWeightMode=RCTWM_WeightedBoth
           time=0
           value=0
          }
          {interpMode=RCIM_Cubic, tangentMode=RCTM_User, time=0.5, value=1}
          {arriveTangentWeight=0.1667, interpMode=RCIM_Cubic, tangentMode=RCTM_User, tangentWeightMode=RCTWM_WeightedBoth, time=1, value=0}
         ]}}})
    )
    Scale RGB=&MaskVectorBySpawnGroup(Masked Vector=<1.2 1.2 1.2>, Passthrough Vector=<0.75 0.75 0.75>)
    ScaleA=true
    ScaleRGB=true
   )
   &DynamicMaterialParameters(
    Index 0 Param 1=&RandomRangeFloat(Minimum=0, Maximum=1)
    Index 0 Param 2=&RandomRangeFloat(Minimum=0, Maximum=1)
    Index 0 Param 3=1
    Index 0 Param 4=&RandomRangeFloat(Minimum=0, Maximum=1)
    Index 1 Param 1=1
    Index 1 Param 2=1
    Index 1 Param 3=1
    Index 1 Param 4=1
    Index 2 Param 1=0.45
    Module.Param0WriteEnabled=true
    Module.Param0WriteEnabled1=true
    Module.Param0WriteEnabled2=true
    Module.Param1WriteEnabled=true
    Module.Param1WriteEnabled1=true
    Module.Param1WriteEnabled2=false
    Module.Param2WriteEnabled=true
    Module.Param2WriteEnabled1=true
    Module.Param2WriteEnabled2=false
    Module.Param3WriteEnabled=true
    Module.Param3WriteEnabled1=true
    Module.Param3WriteEnabled2=false
    Write Parameter Index 1=true
    Write Parameter Index 2=true
   ) {module=DynamicMaterialParameters001}
   &ScaleMeshSize(Scale Factor=&MakeVector(
     X=&FloatFromCurve(FloatCurve=DataInterface() {propertyValues {Curve {keys [<at 0 1>,<at 1 1.5>]}}})
     Y=&FloatFromCurve(FloatCurve=DataInterface() {propertyValues {Curve {keys [<at 0 1>,<at 1 1.35>]}}})
     Z=1
    ))
   &CameraOffset(Camera Offset Amount=&RandomRangeFloat(Minimum=10, Maximum=20)) {module=CameraOffset001}
   &SpriteFacingAndAlignment(WriteAlignment=false, WriteFacing=true)
   &ScaleSpriteSize()
  ]
  rendererClasses [&NiagaraSpriteRendererProperties]
  renderers [{rendererClass=&NiagaraSpriteRendererProperties}]
 }
 Head_Fire_FaceZ {
  emitterUpdateScript [
   &EmitterState(Inactive Response=::NewEnumerator1, Life Cycle Mode=::NewEnumerator1, UseLoopDelay=false)
   &SpawnRate(SpawnRate=5, Use Spawn Probability=false)
   /SetVariables(Emitter.AngleFade=10) {isSetParameters=true}
  ]
  particleSpawnScript [
   &FN_SetSpawnGroup(Number of Spawn Groups=2)
   &InitializeParticle(
    Color=<rgba 0 0.7155 1 0.25>
    Lifetime Max=3.5
    Lifetime Min=2.5
    Lifetime Mode=::NewEnumerator1
    Material Random:NiagaraFloat=&RandomRangeFloat(Minimum=0, Maximum=100)
    Mesh Scale=&Multiply_Vector(
     A=&RandomRangeVector(Minimum=<0.75 0.75 0.25>, Maximum=<1 0.75 0.25>)
     B=&VectorFromFloat(Value=0.2)
    )
    Mesh Scale Mode=::NewEnumerator4
    Sprite Rotation Angle=0.5
    Sprite Rotation Mode=::NewEnumerator1
    Sprite Size Mode=::NewEnumerator3
    Uniform Sprite Size=&Multiply_Float(A=25, B=&RandomRangeFloat(Minimum=0.75, Maximum=1))
    UsePositionOffset=false
    Write Material Random=true
   ) {module=InitializeParticle001}
   &DynamicMaterialParameters(
    Index 0 Param 1=&RandomRangeFloat(Minimum=0, Maximum=0.6)
    Index 0 Param 2=1
    Index 0 Param 3=30
    Module.Param0WriteEnabled=true
    Module.Param1WriteEnabled=true
    Module.Param2WriteEnabled=true
    Module.Param3WriteEnabled=true
   )
   &AddVelocity(Velocity=<0 0 1>)
   &CameraOffset(Camera Offset Amount=3)
   /SetVariables(Particles.Scale=&Multiply_Vector(
     A=$Particles.Scale
     B=&MakeCustomVectorFromBool(Bool=&RandomBool(), True Vector=<1 1 1>, False Vector=<-1 1 1>)
    )) {isSetParameters=true, module=SetVariables_#2}
  ]
  particleUpdateScript [
   &ParticleState()
   &InterpolateOverTime(Rate Of Change=10, Target Vector Value:NiagaraPosition=$Particles.Position, Variable Type=::NewEnumerator2)
   &SkeletalMeshLocation(
    Bone / Socket Index=&SelectIntFromArray(Direct Array Index=$Particles.SpawnGroup, Int Selection Array=DataInterface() {propertyValues {IntData [0,1]}})
    MaskSpawnGroup=false
    Mesh Sampling Type=::NewEnumerator5
    Orientation Sampling=::NewEnumerator1
    Sampled Position Offset=&Add_Vector(
     A=&Add_Vector(A=<-5 3 10>, B=&RandomRangeVector(Minimum=<-2 -6 -12>, Maximum=<10 -2 -5>))
     B=&MakeVector(
      X=&FloatFromCurve(FloatCurve=DataInterface() {propertyValues {Curve {keys [<at 0 0>,<at 1 1>]}}}, Scale Curve=20)
      Y=0
      Z=0
     )
    )
    Skeletal Mesh=$User.Fire_SM
    Socket Sampling Mode=::NewEnumerator5
    Use Sampled Mesh Attribute as Ribbon ID=false
   ) {module=SkeletalMeshLocation001}
   /SetVariables(Particles.Interpolate Dynamic=&MakeFloat4FromVectorAndFloat("Vector (RGB)"=$Output.InterpolateOverTime.Moving Average, "Float (Alpha)"=0.5)) {isSetParameters=true, module=SetVariables_#3}
   &Drag(Drag=0.5, Use Linear Drag=true, Use Rotational Drag=false)
   &SolveForcesAndVelocity(Clamp Velocity=false, Limit Acceleration=false)
   &ScaleColor(
    Scale Alpha=&Multiply_Float(
     A=&FNTimeOfDayScalar(Night Scale=0.5, Morning Scale=0.7, Evening Scale=0.7)
     B=&FloatFromCurve(FloatCurve=DataInterface() {propertyValues {Curve {keys [
          {
           arriveTangentWeight=0.3952
           interpMode=RCIM_Cubic
           leaveTangentWeight=0.3952
           tangentMode=RCTM_User
           tangentWeightMode=RCTWM_WeightedBoth
           time=0
           value=0
          }
          {interpMode=RCIM_Cubic, tangentMode=RCTM_User, time=0.5, value=1}
          {arriveTangentWeight=0.1667, interpMode=RCIM_Cubic, tangentMode=RCTM_User, tangentWeightMode=RCTWM_WeightedBoth, time=1, value=0}
         ]}}})
    )
    Scale RGB=&MaskVectorBySpawnGroup(Masked Vector=<1.2 1.2 1.2>, Passthrough Vector=<0.75 0.75 0.75>)
    ScaleA=true
    ScaleRGB=true
   )
   &DynamicMaterialParameters(
    Index 0 Param 1=&RandomRangeFloat(Minimum=0, Maximum=1)
    Index 0 Param 2=&RandomRangeFloat(Minimum=0, Maximum=1)
    Index 0 Param 3=1
    Index 0 Param 4=&RandomRangeFloat(Minimum=0, Maximum=1)
    Index 1 Param 1=1
    Index 1 Param 2=1
    Index 1 Param 3=1
    Index 1 Param 4=1
    Index 2 Param 1=0.45
    Module.Param0WriteEnabled=true
    Module.Param0WriteEnabled1=true
    Module.Param0WriteEnabled2=true
    Module.Param1WriteEnabled=true
    Module.Param1WriteEnabled1=true
    Module.Param1WriteEnabled2=false
    Module.Param2WriteEnabled=true
    Module.Param2WriteEnabled1=true
    Module.Param2WriteEnabled2=false
    Module.Param3WriteEnabled=true
    Module.Param3WriteEnabled1=true
    Module.Param3WriteEnabled2=false
    Write Parameter Index 1=true
    Write Parameter Index 2=true
   ) {module=DynamicMaterialParameters001}
   &ScaleMeshSize(Scale Factor=&MakeVector(
     X=&FloatFromCurve(FloatCurve=DataInterface() {propertyValues {Curve {keys [<at 0 1>,<at 1 1.5>]}}})
     Y=&FloatFromCurve(FloatCurve=DataInterface() {propertyValues {Curve {keys [<at 0 1>,<at 1 1.35>]}}})
     Z=1
    ))
   &CameraOffset(Camera Offset Amount=&RandomRangeFloat(Minimum=10, Maximum=20)) {module=CameraOffset001}
   &SpriteFacingAndAlignment(Sprite Facing=<0 0 1>, WriteAlignment=false, WriteFacing=true)
   &ScaleSpriteSize()
  ]
  rendererClasses [&NiagaraSpriteRendererProperties]
  renderers [{rendererClass=&NiagaraSpriteRendererProperties}]
 }
 Head_Smoke_FaceX {
  emitterUpdateScript [
   &EmitterState(Inactive Response=::NewEnumerator1, Life Cycle Mode=::NewEnumerator1, UseLoopDelay=false)
   &SpawnRate(SpawnRate=1.5, Use Spawn Probability=false)
  ]
  particleSpawnScript [
   &FN_SetSpawnGroup(Number of Spawn Groups=2)
   &InitializeParticle(
    Color=<rgba 0 0.07 0.1 0.5>
    Lifetime Max=3.5
    Lifetime Min=2.5
    Lifetime Mode=::NewEnumerator1
    Material Random:NiagaraFloat=&RandomRangeFloat(Minimum=0, Maximum=100)
    Mesh Scale=&Multiply_Vector(
     A=&RandomRangeVector(Minimum=<0.75 0.75 0.25>, Maximum=<1 0.75 0.25>)
     B=&VectorFromFloat(Value=0.2)
    )
    Mesh Scale Mode=::NewEnumerator4
    Sprite Rotation Angle=0.5
    Sprite Rotation Mode=::NewEnumerator1
    Sprite Size Mode=::NewEnumerator3
    Uniform Sprite Size=&Multiply_Float(A=50, B=&RandomRangeFloat(Minimum=0.75, Maximum=1))
    UsePositionOffset=false
    Write Material Random=true
   ) {module=InitializeParticle001}
   &DynamicMaterialParameters(
    Index 0 Param 1=&RandomRangeFloat(Minimum=0, Maximum=0.6)
    Index 0 Param 2=1
    Index 0 Param 3=30
    Module.Param0WriteEnabled=true
    Module.Param1WriteEnabled=true
    Module.Param2WriteEnabled=true
    Module.Param3WriteEnabled=true
   )
   &AddVelocity(Velocity=<0 0 1>)
   &CameraOffset(Camera Offset Amount=3)
   /SetVariables(Particles.Scale=&Multiply_Vector(
     A=$Particles.Scale
     B=&MakeCustomVectorFromBool(Bool=&RandomBool(), True Vector=<1 1 1>, False Vector=<-1 1 1>)
    )) {isSetParameters=true, module=SetVariables_#2}
  ]
  particleUpdateScript [
   &ParticleState()
   &InterpolateOverTime(Rate Of Change=10, Target Vector Value:NiagaraPosition=$Particles.Position, Variable Type=::NewEnumerator2)
   &SkeletalMeshLocation(
    Bone / Socket Index=&SelectIntFromArray(Direct Array Index=$Particles.SpawnGroup, Int Selection Array=DataInterface() {propertyValues {IntData [0,1]}})
    MaskSpawnGroup=false
    Mesh Sampling Type=::NewEnumerator5
    Orientation Sampling=::NewEnumerator1
    Sampled Position Offset=&Add_Vector(
     A=&Add_Vector(A=<10 3 10>, B=&RandomRangeVector(Minimum=<-2 -6 -12>, Maximum=<10 -2 -5>))
     B=&MakeVector(
      X=&FloatFromCurve(FloatCurve=DataInterface() {propertyValues {Curve {keys [<at 0 0>,<at 1 1>]}}}, Scale Curve=20)
      Y=0
      Z=0
     )
    )
    Skeletal Mesh=$User.Fire_SM
    Socket Sampling Mode=::NewEnumerator5
    Use Sampled Mesh Attribute as Ribbon ID=false
   ) {module=SkeletalMeshLocation001}
   /SetVariables(Particles.Interpolate Dynamic=&MakeFloat4FromVectorAndFloat("Vector (RGB)"=$Output.InterpolateOverTime.Moving Average, "Float (Alpha)"=0.5)) {isSetParameters=true, module=SetVariables_#3}
   &Drag(Drag=0.5, Use Linear Drag=true, Use Rotational Drag=false)
   &SolveForcesAndVelocity(Clamp Velocity=false, Limit Acceleration=false)
   &ScaleColor(
    Scale Alpha=&Multiply_Float(
     A=&FNTimeOfDayScalar(Night Scale=0.5, Morning Scale=0.7, Evening Scale=0.7)
     B=&FloatFromCurve(FloatCurve=DataInterface() {propertyValues {Curve {keys [
          {
           arriveTangentWeight=0.3952
           interpMode=RCIM_Cubic
           leaveTangentWeight=0.3952
           tangentMode=RCTM_User
           tangentWeightMode=RCTWM_WeightedBoth
           time=0
           value=1
          }
          {interpMode=RCIM_Cubic, tangentMode=RCTM_User, time=0.5, value=1}
          {arriveTangentWeight=0.1667, interpMode=RCIM_Cubic, tangentMode=RCTM_User, tangentWeightMode=RCTWM_WeightedBoth, time=1, value=0}
         ]}}})
    )
    Scale RGB=&MaskVectorBySpawnGroup(Masked Vector=<0.01 0.01 0.01>, Passthrough Vector=<0.02 0.02 0.02>)
    ScaleA=true
    ScaleRGB=true
   )
   &DynamicMaterialParameters(
    Index 0 Param 1=&RandomRangeFloat(Minimum=0, Maximum=1)
    Index 0 Param 2=&RandomRangeFloat(Minimum=0, Maximum=1)
    Index 0 Param 3=1
    Index 0 Param 4=&RandomRangeFloat(Minimum=0, Maximum=1)
    Index 1 Param 1=1
    Index 1 Param 2=1
    Index 1 Param 3=1
    Index 1 Param 4=1
    Index 2 Param 1=0.45
    Module.Param0WriteEnabled=true
    Module.Param0WriteEnabled1=true
    Module.Param0WriteEnabled2=true
    Module.Param1WriteEnabled=true
    Module.Param1WriteEnabled1=true
    Module.Param1WriteEnabled2=false
    Module.Param2WriteEnabled=true
    Module.Param2WriteEnabled1=true
    Module.Param2WriteEnabled2=false
    Module.Param3WriteEnabled=true
    Module.Param3WriteEnabled1=true
    Module.Param3WriteEnabled2=false
    Write Parameter Index 1=true
    Write Parameter Index 2=true
   ) {module=DynamicMaterialParameters001}
   &ScaleMeshSize(Scale Factor=&MakeVector(
     X=&FloatFromCurve(FloatCurve=DataInterface() {propertyValues {Curve {keys [<at 0 1>,<at 1 1.5>]}}})
     Y=&FloatFromCurve(FloatCurve=DataInterface() {propertyValues {Curve {keys [<at 0 1>,<at 1 1.35>]}}})
     Z=1
    ))
   &CameraOffset(Camera Offset Amount=&RandomRangeFloat(Minimum=10, Maximum=20)) {module=CameraOffset001}
   &SpriteFacingAndAlignment(WriteAlignment=false, WriteFacing=true)
   &ScaleSpriteSize()
   &ScaleColor(Scale Alpha=&OneMinusFloat(Float=$System.Timeline.NormalizedTime), ScaleA=true, ScaleRGB=false) {module=ScaleColor001}
  ]
  rendererClasses [&NiagaraMeshRendererProperties, &NiagaraSpriteRendererProperties]
  renderers [
   {rendererClass=&NiagaraMeshRendererProperties, rendererIndex=0}
   {rendererClass=&NiagaraSpriteRendererProperties, rendererIndex=1}
  ]
 }
 Head_Smoke_FaceX001 {
  emitterUpdateScript [
   &EmitterState(Inactive Response=::NewEnumerator1, Life Cycle Mode=::NewEnumerator1, UseLoopDelay=false)
   &SpawnRate(SpawnRate=1.5, Use Spawn Probability=false)
  ]
  particleSpawnScript [
   &FN_SetSpawnGroup(Number of Spawn Groups=2)
   &InitializeParticle(
    Color=<rgba 0 0.07 0.1 0.5>
    Lifetime Max=3.5
    Lifetime Min=2.5
    Lifetime Mode=::NewEnumerator1
    Material Random:NiagaraFloat=&RandomRangeFloat(Minimum=0, Maximum=100)
    Mesh Scale=&Multiply_Vector(
     A=&RandomRangeVector(Minimum=<0.75 0.75 0.25>, Maximum=<1 0.75 0.25>)
     B=&VectorFromFloat(Value=0.2)
    )
    Mesh Scale Mode=::NewEnumerator4
    Sprite Rotation Angle=0.5
    Sprite Rotation Mode=::NewEnumerator1
    Sprite Size Mode=::NewEnumerator3
    Uniform Sprite Size=&Multiply_Float(A=50, B=&RandomRangeFloat(Minimum=0.75, Maximum=1))
    UsePositionOffset=false
    Write Material Random=true
   ) {module=InitializeParticle001}
   &DynamicMaterialParameters(
    Index 0 Param 1=&RandomRangeFloat(Minimum=0, Maximum=0.6)
    Index 0 Param 2=1
    Index 0 Param 3=30
    Module.Param0WriteEnabled=true
    Module.Param1WriteEnabled=true
    Module.Param2WriteEnabled=true
    Module.Param3WriteEnabled=true
   )
   &AddVelocity(Velocity=<0 0 1>)
   &CameraOffset(Camera Offset Amount=3)
   /SetVariables(Particles.Scale=&Multiply_Vector(
     A=$Particles.Scale
     B=&MakeCustomVectorFromBool(Bool=&RandomBool(), True Vector=<1 1 1>, False Vector=<-1 1 1>)
    )) {isSetParameters=true, module=SetVariables_#2}
  ]
  particleUpdateScript [
   &ParticleState()
   &InterpolateOverTime(Rate Of Change=10, Target Vector Value:NiagaraPosition=$Particles.Position, Variable Type=::NewEnumerator2)
   &SkeletalMeshLocation(
    Bone / Socket Index=&SelectIntFromArray(Direct Array Index=$Particles.SpawnGroup, Int Selection Array=DataInterface() {propertyValues {IntData [0,1]}})
    MaskSpawnGroup=false
    Mesh Sampling Type=::NewEnumerator5
    Orientation Sampling=::NewEnumerator1
    Sampled Position Offset=&Add_Vector(
     A=&Add_Vector(A=<10 3 10>, B=&RandomRangeVector(Minimum=<-2 -6 -12>, Maximum=<10 -2 -5>))
     B=&MakeVector(
      X=&FloatFromCurve(FloatCurve=DataInterface() {propertyValues {Curve {keys [<at 0 0>,<at 1 1>]}}}, Scale Curve=20)
      Y=0
      Z=0
     )
    )
    Skeletal Mesh=$User.Fire_SM
    Socket Sampling Mode=::NewEnumerator5
    Use Sampled Mesh Attribute as Ribbon ID=false
   ) {module=SkeletalMeshLocation001}
   /SetVariables(Particles.Interpolate Dynamic=&MakeFloat4FromVectorAndFloat("Vector (RGB)"=$Output.InterpolateOverTime.Moving Average, "Float (Alpha)"=0.5)) {isSetParameters=true, module=SetVariables_#3}
   &Drag(Drag=0.5, Use Linear Drag=true, Use Rotational Drag=false)
   &SolveForcesAndVelocity(Clamp Velocity=false, Limit Acceleration=false)
   &ScaleColor(
    Scale Alpha=&Multiply_Float(
     A=&FNTimeOfDayScalar(Night Scale=0.5, Morning Scale=0.7, Evening Scale=0.7)
     B=&FloatFromCurve(FloatCurve=DataInterface() {propertyValues {Curve {keys [
          {
           arriveTangentWeight=0.3952
           interpMode=RCIM_Cubic
           leaveTangentWeight=0.3952
           tangentMode=RCTM_User
           tangentWeightMode=RCTWM_WeightedBoth
           time=0
           value=1
          }
          {interpMode=RCIM_Cubic, tangentMode=RCTM_User, time=0.5, value=1}
          {arriveTangentWeight=0.1667, interpMode=RCIM_Cubic, tangentMode=RCTM_User, tangentWeightMode=RCTWM_WeightedBoth, time=1, value=0}
         ]}}})
    )
    Scale RGB=&MaskVectorBySpawnGroup(Masked Vector=<0.01 0.01 0.01>, Passthrough Vector=<0.02 0.02 0.02>)
    ScaleA=true
    ScaleRGB=true
   )
   &DynamicMaterialParameters(
    Index 0 Param 1=&RandomRangeFloat(Minimum=0, Maximum=1)
    Index 0 Param 2=&RandomRangeFloat(Minimum=0, Maximum=1)
    Index 0 Param 3=1
    Index 0 Param 4=&RandomRangeFloat(Minimum=0, Maximum=1)
    Index 1 Param 1=1
    Index 1 Param 2=1
    Index 1 Param 3=1
    Index 1 Param 4=1
    Index 2 Param 1=0.45
    Module.Param0WriteEnabled=true
    Module.Param0WriteEnabled1=true
    Module.Param0WriteEnabled2=true
    Module.Param1WriteEnabled=true
    Module.Param1WriteEnabled1=true
    Module.Param1WriteEnabled2=false
    Module.Param2WriteEnabled=true
    Module.Param2WriteEnabled1=true
    Module.Param2WriteEnabled2=false
    Module.Param3WriteEnabled=true
    Module.Param3WriteEnabled1=true
    Module.Param3WriteEnabled2=false
    Write Parameter Index 1=true
    Write Parameter Index 2=true
   ) {module=DynamicMaterialParameters001}
   &ScaleMeshSize(Scale Factor=&MakeVector(
     X=&FloatFromCurve(FloatCurve=DataInterface() {propertyValues {Curve {keys [<at 0 1>,<at 1 1.5>]}}})
     Y=&FloatFromCurve(FloatCurve=DataInterface() {propertyValues {Curve {keys [<at 0 1>,<at 1 1.35>]}}})
     Z=1
    ))
   &CameraOffset(Camera Offset Amount=&RandomRangeFloat(Minimum=10, Maximum=20)) {module=CameraOffset001}
   &SpriteFacingAndAlignment(Sprite Facing=<0 0 1>, WriteAlignment=false, WriteFacing=true)
   &ScaleSpriteSize()
   &ScaleColor(Scale Alpha=&OneMinusFloat(Float=$System.Timeline.NormalizedTime), ScaleA=true, ScaleRGB=false) {module=ScaleColor001}
  ]
  rendererClasses [&NiagaraMeshRendererProperties, &NiagaraSpriteRendererProperties]
  renderers [
   {rendererClass=&NiagaraMeshRendererProperties, rendererIndex=0}
   {rendererClass=&NiagaraSpriteRendererProperties, rendererIndex=1}
  ]
 }
 Ribbons {
  emitterUpdateScript [
   &EmitterState(Inactive Response=::NewEnumerator1, Life Cycle Mode=::NewEnumerator1, Loop Duration=2, UseLoopDelay=false)
   &SpawnRate(
    SpawnRate=&Multiply_FloatByInt(Float=&Multiply_Float(A=12, B=1), Integer=8)
    Use Spawn Probability=false
   )
  ]
  particleSpawnScript [
   /SetVariables(Particles.Ribbon:NiagaraInt32=&ModuloInt(A=&ReturnParticleIDAttributes(), B=9)) {isSetParameters=true}
   /SetVariables(Particles.RibbonID=/Niagara/DynamicInputs/NiagaraID/MakeNiagaraIDFromInt(Index:NiagaraInt32=$Particles.Ribbon, Acquire Tag:NiagaraInt32=0)) {isSetParameters=true}
   &InitializeRibbon(
    Color Maximum:LinearColor=<rgba 0.0075 0.0075 0.0075 1>
    Color Minimum:LinearColor=<rgba 0.0075 0.0075 0.0075 1>
    Color Mode:ENiagara_ColorInitializationMode=::NewEnumerator2
    Lifetime=1.75
    UsePositionOffset=false
    Write Material Random=true
   )
   &SkeletalMeshLocation(
    Bone / Socket Index=&SelectIntFromArray(
     Direct Array Index=$Particles.Ribbon
     Direct Array Mode=::NewEnumerator1
     Int Selection Array=DataInterface() {propertyValues {IntData [2,3,4,5,6,7,8,9,10]}}
    )
    MaskSpawnGroup=false
    Mesh Sampling Type=::NewEnumerator5
    Position Sampling=::NewEnumerator1
    Sampled Position Offset=&SelectVectorFromArray(Direct Array Index=$Particles.Ribbon, Vector Selection Array=DataInterface() {propertyValues {FloatData [<3 0 0>, <0 0 0>]}})
    Skeletal Mesh=$User.SkeletalMesh
    Socket Sampling Mode=::NewEnumerator5
    Use Sampled Mesh Attribute as Ribbon ID=false
   )
   &AddVelocityInCone(
    Cone Angle=0
    Cone Axis=&ReturnMeshOrientationAxis(Mesh Orientation Axis=::XAxis, Mesh Orientation Quat=$Particles.SkeletalMeshLocation.SampledRotation)
    Cone Axis Coordinate Space=::World
    Module.Use Velocity Falloff On Cone Axis=false
    Velocity Distribution Along Cone Axis=0.25
    Velocity Strength=&Multiply_Float(A=10, B=$User.IdleRibbonVelocityMult)
   )
   &AddVelocityFromPoint(
    Module.Use Falloff Distance=false
    Origin Offset=<0 -10 37>
    Velocity Strength=&Multiply_Float(
     A=&Multiply_Float(A=4, B=$User.IdleRibbonVelocityMult)
     B=&SelectFloatFromArray(Direct Array Index=$Particles.Ribbon, Float Selection Array=DataInterface() {propertyValues {FloatData [0.5,1,1,1,1,1,1,-0.25,-0.5]}})
    )
   )
   &AddVelocity(
    Coordinate Space=::World
    Velocity=&Multiply_VectorByFloat(
     Vector=&ReturnMeshOrientationAxis(Mesh Orientation Axis=::YAxis, Mesh Orientation Quat=$Particles.SkeletalMeshLocation.SampledRotation)
     Float=&Multiply_Float(
      A=3
      B=&SelectFloatFromArray(Direct Array Index=$Particles.Ribbon, Float Selection Array=DataInterface() {propertyValues {FloatData [0,0,0,1,1,1,1,0,0]}})
     )
    )
   )
   &AddVelocity(Velocity=<0 -5 0>) {module=AddVelocity001}
  ]
  particleUpdateScript [
   &ParticleState()
   &ScaleColor(
    Scale Alpha=&Multiply_Float(
     A=&FloatFromCurve(
      FloatCurve=DataInterface() {propertyValues {Curve {keys [{interpMode=RCIM_Cubic, tangentMode=RCTM_User, time=0.6, value=1}, {interpMode=RCIM_Cubic, tangentMode=RCTM_User, time=1, value=0}]}}}
      Scale Curve=0.03
     )
     B=&OneMinusFloat(Float=$System.Timeline.NormalizedTime)
    )
    ScaleA=true
    ScaleRGB=true
   )
   &GravityForce(Coordinate Space=::Local, Gravity=<0 -5 0>)
   &CurlNoiseForce(
    Mask Curl Noise=true
    Noise Frequency=120
    Noise Strength=&FloatFromCurve(
     FloatCurve=DataInterface() {propertyValues {Curve {keys [
         {arriveTangent=0.9574, interpMode=RCIM_Cubic, leaveTangent=0.9574, leaveTangentWeight=0.7125, tangentMode=RCTM_User, time=0, value=0}
         {
          arriveTangentWeight=0.8409
          interpMode=RCIM_Cubic
          leaveTangentWeight=0.8409
          tangentMode=RCTM_User
          tangentWeightMode=RCTWM_WeightedBoth
          time=1
          value=1
         }
        ]}}}
     Scale Curve=30
    )
    Pan Noise=true
    Pan Noise Field=&RandomVector(Vector Scale=1)
   )
   /Niagara/Modules/Update/Forces/LineAttractionForce(
    Attraction Strength:NiagaraFloat=&FloatFromCurve(
     FloatCurve=DataInterface() {propertyValues {Curve {keys [{tangentMode=RCTM_User, time=0, value=0}, {tangentMode=RCTM_User, time=1, value=1}]}}}
     Scale Curve=30
    )
    Line End:Vector3f=<0 -120 65>
    Line Segment in Localspace=true
    Line Start:Vector3f=<0 0 37>
    Use Attraction Falloff=false
   )
   &InheritSourceMovement(
    Applied Position Scale=&VectorFromFloat(Value=&FloatFromCurve(FloatCurve=DataInterface() {propertyValues {Curve {keys [<at 0 1>,<at 0.1 0>]}}}))
    ApplyForce=false
    ApplyPosition=true
    ApplyPositionFalloff=false
   )
   &SolveForcesAndVelocity(Clamp Velocity=false, Limit Acceleration=false)
   &ScaleRibbonWidth(Ribbon Width Scale=&FloatFromCurve(
     FloatCurve=DataInterface() {propertyValues {Curve {keys [
         {interpMode=RCIM_Cubic, leaveTangentWeight=0.3532, tangentMode=RCTM_User, time=0, value=1}
         {arriveTangentWeight=0.701, interpMode=RCIM_Cubic, tangentMode=RCTM_User, time=1, value=3}
        ]}}}
     Scale Curve=&Multiply_Float(A=1.2, B=&OneMinusFloat(Float=$System.Timeline.NormalizedTime))
    ))
   /SetVariables(Particles.LerpedPosition:Vector3f=&Lerp_Vector(
     A=$Particles.Initial.SkeletalMeshLocation.SampledPosition
     B=Linked() {linkedVariable {name=Particles.Position, type=NiagaraPosition}}
     Alpha=&FloatFromCurve(FloatCurve=DataInterface() {propertyValues {Curve {keys [{tangentMode=RCTM_User, time=0.1, value=0}, {tangentMode=RCTM_User, time=0.15, value=1}]}}})
    )) {isSetParameters=true}
   &KillParticles(Kill Particles=&SetBoolByFloatComparison(A=$System.Timeline.NormalizedTime, B=0.99))
   &ScaleRibbonWidth(Ribbon Width Scale=&FloatFromCurve(FloatCurve=DataInterface() {propertyValues {Curve {keys [
         {interpMode=RCIM_Cubic, tangentMode=RCTM_User, time=0, value=0}
         {interpMode=RCIM_Cubic, tangentMode=RCTM_User, time=0.3, value=1}
         {arriveTangent=-0.0574, interpMode=RCIM_Cubic, leaveTangent=-0.0574, tangentMode=RCTM_User, time=0.5, value=1}
         {arriveTangent=-0.2625, interpMode=RCIM_Cubic, leaveTangent=-0.2625, tangentMode=RCTM_User, time=1, value=0}
        ]}}})) {module=ScaleRibbonWidth001}
   /SetVariables(Particles.RibbonFacing=<0 0 1>) {isSetParameters=true}
  ]
  rendererClasses [&NiagaraRibbonRendererProperties]
  renderers [{rendererClass=&NiagaraRibbonRendererProperties}]
 }
}
summary {
 systemName=NS_Dark_Pit_Pick_IDle
 emitters [
  {emitterName=Ribbons, rendererClasses [&NiagaraRibbonRendererProperties]}
  {emitterName=Head_Blue_Smoke, rendererClasses [&NiagaraSpriteRendererProperties]}
  {emitterName=Attached_Ribbon, rendererClasses [&NiagaraRibbonRendererProperties]}
  {emitterName=Head_Fire_FaceX, rendererClasses [&NiagaraSpriteRendererProperties]}
  {emitterName=Head_Fire_FaceZ, rendererClasses [&NiagaraSpriteRendererProperties]}
  {
   emitterName=Head_Smoke_FaceX
   rendererClasses [&NiagaraMeshRendererProperties, &NiagaraSpriteRendererProperties]
  }
  {
   emitterName=Head_Smoke_FaceX001
   rendererClasses [&NiagaraMeshRendererProperties, &NiagaraSpriteRendererProperties]
  }
  {emitterName=Attached_Ribbon001, rendererClasses [&NiagaraRibbonRendererProperties]}
  {emitterName=Attached_Ribbon002, rendererClasses [&NiagaraRibbonRendererProperties]}
  {emitterName=Attached_Ribbon003, rendererClasses [&NiagaraRibbonRendererProperties]}
  {emitterName=Goop, rendererClasses [&NiagaraMeshRendererProperties]}
 ]
 userVariables [
  User.AxeHead_SM {
   type=NiagaraDataInterfaceSkeletalMesh
   defaultValue {
    struct=NiagaraExt_VariableValue_DataInterface
    value {dataInterface=&self:NiagaraDataInterfaceSkeletalMesh_1}
   }
  }
  User.Fire_SM {
   type=NiagaraDataInterfaceSkeletalMesh
   defaultValue {
    struct=NiagaraExt_VariableValue_DataInterface
    value {dataInterface=&self:NiagaraDataInterfaceSkeletalMesh_2}
   }
  }
  User.Goop_SM {
   type=NiagaraDataInterfaceSkeletalMesh
   defaultValue {
    struct=NiagaraExt_VariableValue_DataInterface
    value {dataInterface=&self:NiagaraDataInterfaceSkeletalMesh_4}
   }
  }
  User.IdleRibbonVelocityMult {type=NiagaraFloat, defaultValue=0.8, description=""}
  User.RibbonIDOffset {type=NiagaraInt32}
  User.RunningLifetime {type=NiagaraFloat, defaultValue=1.5, description=""}
  User.RunningOpacity {type=NiagaraFloat, defaultValue=0.1, description=""}
  User.SkeletalMesh {
   type=NiagaraDataInterfaceSkeletalMesh
   defaultValue {
    struct=NiagaraExt_VariableValue_DataInterface
    value {dataInterface=&self:NiagaraDataInterfaceSkeletalMesh_0}
   }
  }
  User.Smoke_SM {
   type=NiagaraDataInterfaceSkeletalMesh
   defaultValue {
    struct=NiagaraExt_VariableValue_DataInterface
    value {dataInterface=&self:NiagaraDataInterfaceSkeletalMesh_3}
   }
  }
 ]
}
