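# Niagara swing trail for the Dark Pit male pickaxe, with one active AnimTrail emitter and two disabled variants.
# The emitter spawns short-lived cyan ribbon particles through AnimTrail bone/socket modules, with distance-based UVs and opposing twist values.
# Three ribbon renderers draw it while alpha and width curves shape the fade.
# A wider dark smoke color layer also fades, with an alpha boost at night.
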NiagaraSystem /BRCosmetics/Weapons/FORT_Melee/Pickaxe_Dark_Pit_Male/FX/NS_Dark_Pit_Pick_Swing
dependencies {
 usedDataInterfaces [&NiagaraDataInterfaceCurve, &NiagaraDataInterfaceColorCurve]
 usedDynamicInputs []
 usedModules [&SystemState]
}
emitters {
 FN_AnimTrailMain {
  emitterUpdateScript [
   &EmitterState(Life Cycle Mode=::NewEnumerator1, Loop Behavior=::NewEnumerator1, Loop Duration=0.3, UseLoopDelay=false)
   &FN_SpawnAnimTrail_v2()
   &SpawnPerUnit(Module.Use Max Movement Threshold=true, Module.Use Movement Tolerance=false, Spawn Probability=0, Use Spawn Probability=true)
  ]
  particleSpawnScript [
   /Game/Effects/Niagara/Modules/AnimTrail/FN_InitializeAnimTrail/FN_InitializeAnimTrailBoneIndecies(First Bone or Socket Index=1)
   /Game/Effects/Niagara/Modules/AnimTrail/FN_InitializeAnimTrail/FN_InitializeAnimTrail_v3/FN_InitializeAnimTrail(Color:LinearColor=<rgba 0 0.5934 0.8 1>, Lifetime=0.25)
   /SetVariables(
    Particles.Smoke Color:LinearColor=<rgba 0.015 0.015 0.015 1.5>
    Particles.Smoke Width:NiagaraFloat=&Multiply_Float(A=$Particles.RibbonWidth, B=1.5)
   ) {isSetParameters=true, module=SetVariables_#1}
   &AddVelocity(Velocity=<0 -2500 0>) {enabled=false}
   /SetVariables(Particles.RibbonUVDistance:NiagaraFloat=$Emitter.DistanceTraveled) {isSetParameters=true, module=SetVariables_#2}
   /SetVariables(Particles.RibbonTwist=10, Particles.RibbonTwist2=-10) {isSetParameters=true, module=SetVariables_#3}
  ]
  particleUpdateScript [
   &SolveForcesAndVelocity(Clamp Velocity=false, Limit Acceleration=false)
   /Game/Effects/Niagara/Modules/AnimTrail/FN_UpdateAnimTrail/FN_UpdateAnimTrail_v3/FN_UpdateAnimTrail(Locked Particle RibbonWidth:NiagaraFloat=$Particles.RibbonWidth)
   &ScaleColor(
    Scale Alpha=&Multiply_Float(
     A=&Multiply_Float(
      A=&FloatFromCurve(
       FloatCurve=DataInterface() {propertyValues {Curve {keys [
           {arriveTangent=-4.74, arriveTangentWeight=0.6396, leaveTangentWeight=0.6396, time=0.5, value=2}
           {arriveTangent=0.0404, arriveTangentWeight=0.614, leaveTangent=0.0404, leaveTangentWeight=0.614, time=1, value=0}
          ]}}}
       Scale Curve=0.25
      )
      B=&FNTimeOfDayScalar(Night Scale=1, Morning Scale=1, Evening Scale=1)
     )
     B=1
    )
    ScaleA=true
    ScaleRGB=true
   )
   &Color(Color=&MultiplyLinearColor(
     Linear Color=&ColorFromCurve(
      ColorCurve=DataInterface() {propertyValues {
        AlphaCurve {keys [{arriveTangent=-1, leaveTangent=-1, time=0, value=1}, <at 0.9884 1>]}
        BlueCurve {keys [
          {arriveTangentWeight=0.4049, interpMode=RCIM_Cubic, leaveTangentWeight=0.2272, tangentMode=RCTM_Break, time=0.4169, value=1}
          {interpMode=RCIM_Cubic, leaveTangentWeight=0.2686, tangentMode=RCTM_User, time=0.8302, value=0}
         ]}
        GreenCurve {keys [
          {arriveTangentWeight=0.3287, interpMode=RCIM_Cubic, leaveTangentWeight=0.2272, tangentMode=RCTM_Break, time=0.4169, value=0.5536}
          {interpMode=RCIM_Cubic, leaveTangentWeight=0.2086, tangentMode=RCTM_User, time=0.8302, value=0}
         ]}
        RedCurve {keys [
          {arriveTangentWeight=0.1727, interpMode=RCIM_Cubic, leaveTangentWeight=0.1607, tangentMode=RCTM_Break, time=0.4169, value=0.056}
          {interpMode=RCIM_Cubic, leaveTangentWeight=0.1732, tangentMode=RCTM_User, time=0.8302, value=0}
         ]}
       }}
      CurveIndex=&ReturnNormalizedExecIndex()
     )
     Scale Factor=&MakeFloat4FromVectorAndFloat(
      "Vector (RGB)"=<1 1 1>
      "Float (Alpha)"=&FloatFromCurve(
       FloatCurve=DataInterface() {propertyValues {Curve {keys [
           {
            arriveTangent=-4.74
            arriveTangentWeight=0.6396
            interpMode=RCIM_Cubic
            leaveTangent=-4.74
            leaveTangentWeight=0.6396
            tangentMode=RCTM_User
            time=0
            value=2
           }
           {
            arriveTangent=0.0404
            arriveTangentWeight=0.614
            interpMode=RCIM_Cubic
            leaveTangent=0.0404
            leaveTangentWeight=0.614
            tangentMode=RCTM_User
            tangentWeightMode=RCTWM_WeightedBoth
            time=1
            value=0
           }
          ]}}}
       Scale Curve=0.5
      )
     )
    )) {enabled=false}
   &ScaleRibbonWidth(Ribbon Width Scale=&FloatFromCurve(FloatCurve=DataInterface() {propertyValues {Curve {keys [<at 0 1>,<at 1 0.9>]}}}))
   /SetVariables(Particles.Scaled Smoke Color:LinearColor=&MultiplyLinearColor(
     Linear Color=$Particles.Initial.Smoke Color
     Scale Factor=&MakeFloat4FromVectorAndFloat(
      "Vector (RGB)"=<1 1 1>
      "Float (Alpha)"=&Multiply_Float(
       A=&FloatFromCurve(
        FloatCurve=DataInterface() {propertyValues {Curve {keys [{tangentMode=RCTM_User, time=0, value=1}, {tangentMode=RCTM_User, time=1, value=0}]}}}
        Scale Curve=2
       )
       B=&FNTimeOfDayScalar(Night Scale=2.2, Morning Scale=1, Evening Scale=1)
      )
     )
    )) {isSetParameters=true, module=SetVariables_#4}
  ]
  rendererClasses [&NiagaraRibbonRendererProperties]
  renderers [
   {rendererClass=&NiagaraRibbonRendererProperties}
   {rendererClass=&NiagaraRibbonRendererProperties, rendererIndex=1}
   {rendererClass=&NiagaraRibbonRendererProperties, rendererIndex=2}
  ]
 }
 FN_AnimTrailMain001 {
  bEnabled=false
  emitterUpdateScript [
   &EmitterState(Life Cycle Mode=::NewEnumerator1, Loop Behavior=::NewEnumerator1, Loop Duration=0.3, UseLoopDelay=false)
   &FN_SpawnAnimTrail_v2()
   &SpawnPerUnit(Module.Use Max Movement Threshold=true, Module.Use Movement Tolerance=false, Spawn Probability=0, Use Spawn Probability=true)
  ]
  particleSpawnScript [
   &FN_InitializeAnimTrailBoneIndecies(First Bone or Socket Index=1)
   &FN_InitializeAnimTrail(Color:LinearColor=<rgba 0 0.5934 0.8 1>, Lifetime=0.25)
   /SetVariables(
    Particles.Smoke Color:LinearColor=<rgba 0.015 0.015 0.015 1.5>
    Particles.Smoke Width:NiagaraFloat=&Multiply_Float(A=$Particles.RibbonWidth, B=1.5)
   ) {isSetParameters=true, module=SetVariables_#1}
   &AddVelocity(Velocity=<0 -2500 0>) {enabled=false}
   /SetVariables(Particles.RibbonUVDistance:NiagaraFloat=$Emitter.DistanceTraveled) {isSetParameters=true, module=SetVariables_#2}
  ]
  particleUpdateScript [
   &SolveForcesAndVelocity(Clamp Velocity=false, Limit Acceleration=false)
   &FN_UpdateAnimTrail(Locked Particle RibbonWidth:NiagaraFloat=$Particles.RibbonWidth)
   &ScaleColor(
    Scale Alpha=&Multiply_Float(
     A=&Multiply_Float(
      A=&FloatFromCurve(
       FloatCurve=DataInterface() {propertyValues {Curve {keys [
           {arriveTangent=-4.74, arriveTangentWeight=0.6396, leaveTangentWeight=0.6396, time=0.5, value=2}
           {arriveTangent=0.0404, arriveTangentWeight=0.614, leaveTangent=0.0404, leaveTangentWeight=0.614, time=1, value=0}
          ]}}}
       Scale Curve=0.25
      )
      B=&FNTimeOfDayScalar(Night Scale=1, Morning Scale=1, Evening Scale=1)
     )
     B=1
    )
    ScaleA=true
    ScaleRGB=true
   )
   &Color(Color=&MultiplyLinearColor(
     Linear Color=&ColorFromCurve(
      ColorCurve=DataInterface() {propertyValues {
        AlphaCurve {keys [{arriveTangent=-1, leaveTangent=-1, time=0, value=1}, <at 0.9884 1>]}
        BlueCurve {keys [
          {arriveTangentWeight=0.4049, interpMode=RCIM_Cubic, leaveTangentWeight=0.2272, tangentMode=RCTM_Break, time=0.4169, value=1}
          {interpMode=RCIM_Cubic, leaveTangentWeight=0.2686, tangentMode=RCTM_User, time=0.8302, value=0}
         ]}
        GreenCurve {keys [
          {arriveTangentWeight=0.3287, interpMode=RCIM_Cubic, leaveTangentWeight=0.2272, tangentMode=RCTM_Break, time=0.4169, value=0.5536}
          {interpMode=RCIM_Cubic, leaveTangentWeight=0.2086, tangentMode=RCTM_User, time=0.8302, value=0}
         ]}
        RedCurve {keys [
          {arriveTangentWeight=0.1727, interpMode=RCIM_Cubic, leaveTangentWeight=0.1607, tangentMode=RCTM_Break, time=0.4169, value=0.056}
          {interpMode=RCIM_Cubic, leaveTangentWeight=0.1732, tangentMode=RCTM_User, time=0.8302, value=0}
         ]}
       }}
      CurveIndex=&ReturnNormalizedExecIndex()
     )
     Scale Factor=&MakeFloat4FromVectorAndFloat(
      "Vector (RGB)"=<1 1 1>
      "Float (Alpha)"=&FloatFromCurve(
       FloatCurve=DataInterface() {propertyValues {Curve {keys [
           {
            arriveTangent=-4.74
            arriveTangentWeight=0.6396
            interpMode=RCIM_Cubic
            leaveTangent=-4.74
            leaveTangentWeight=0.6396
            tangentMode=RCTM_User
            time=0
            value=2
           }
           {
            arriveTangent=0.0404
            arriveTangentWeight=0.614
            interpMode=RCIM_Cubic
            leaveTangent=0.0404
            leaveTangentWeight=0.614
            tangentMode=RCTM_User
            tangentWeightMode=RCTWM_WeightedBoth
            time=1
            value=0
           }
          ]}}}
       Scale Curve=0.5
      )
     )
    )) {enabled=false}
   &ScaleRibbonWidth(Ribbon Width Scale=&FloatFromCurve(FloatCurve=DataInterface() {propertyValues {Curve {keys [<at 0 1>,<at 1 0.9>]}}}))
   /SetVariables(Particles.Scaled Smoke Color:LinearColor=&MultiplyLinearColor(
     Linear Color=$Particles.Initial.Smoke Color
     Scale Factor=&MakeFloat4FromVectorAndFloat(
      "Vector (RGB)"=<1 1 1>
      "Float (Alpha)"=&Multiply_Float(
       A=&FloatFromCurve(
        FloatCurve=DataInterface() {propertyValues {Curve {keys [{tangentMode=RCTM_User, time=0, value=1}, {tangentMode=RCTM_User, time=1, value=0}]}}}
        Scale Curve=2
       )
       B=&FNTimeOfDayScalar(Night Scale=2.2, Morning Scale=1, Evening Scale=1)
      )
     )
    )) {isSetParameters=true, module=SetVariables_#4}
  ]
  rendererClasses [&NiagaraRibbonRendererProperties]
  renderers [
   {rendererClass=&NiagaraRibbonRendererProperties}
   {rendererClass=&NiagaraRibbonRendererProperties, rendererIndex=1}
  ]
 }
 FN_AnimTrailMain002 {
  bEnabled=false
  emitterUpdateScript [
   &EmitterState(Life Cycle Mode=::NewEnumerator1, Loop Behavior=::NewEnumerator1, Loop Duration=0.3, UseLoopDelay=false)
   &FN_SpawnAnimTrail_v2()
   &SpawnPerUnit(Module.Use Max Movement Threshold=true, Module.Use Movement Tolerance=false, Spawn Probability=0, Use Spawn Probability=true)
  ]
  particleSpawnScript [
   &FN_InitializeAnimTrailBoneIndecies(First Bone or Socket Index=1)
   &FN_InitializeAnimTrail(Color:LinearColor=<rgba 0 0.5934 0.8 1>, Lifetime=0.25)
   /SetVariables(
    Particles.Smoke Color:LinearColor=<rgba 0.015 0.015 0.015 1.5>
    Particles.Smoke Width:NiagaraFloat=&Multiply_Float(A=$Particles.RibbonWidth, B=1.5)
   ) {isSetParameters=true, module=SetVariables_#1}
   &AddVelocity(Velocity=<0 -2500 0>) {enabled=false}
   /SetVariables(Particles.RibbonUVDistance:NiagaraFloat=$Emitter.DistanceTraveled) {isSetParameters=true, module=SetVariables_#2}
   /SetVariables(Particles.RibbonTwist=10, Particles.RibbonTwist2=-10) {isSetParameters=true, module=SetVariables_#3}
  ]
  particleUpdateScript [
   &SolveForcesAndVelocity(Clamp Velocity=false, Limit Acceleration=false)
   &FN_UpdateAnimTrail(Locked Particle RibbonWidth:NiagaraFloat=$Particles.RibbonWidth)
   &ScaleColor(
    Scale Alpha=&Multiply_Float(
     A=&Multiply_Float(
      A=&FloatFromCurve(
       FloatCurve=DataInterface() {propertyValues {Curve {keys [
           {arriveTangent=-4.74, arriveTangentWeight=0.6396, leaveTangentWeight=0.6396, time=0.5, value=2}
           {arriveTangent=0.0404, arriveTangentWeight=0.614, leaveTangent=0.0404, leaveTangentWeight=0.614, time=1, value=0}
          ]}}}
       Scale Curve=0.25
      )
      B=&FNTimeOfDayScalar(Night Scale=1, Morning Scale=1, Evening Scale=1)
     )
     B=1
    )
    ScaleA=true
    ScaleRGB=true
   )
   &Color(Color=&MultiplyLinearColor(
     Linear Color=&ColorFromCurve(
      ColorCurve=DataInterface() {propertyValues {
        AlphaCurve {keys [{arriveTangent=-1, leaveTangent=-1, time=0, value=1}, <at 0.9884 1>]}
        BlueCurve {keys [
          {arriveTangentWeight=0.4049, interpMode=RCIM_Cubic, leaveTangentWeight=0.2272, tangentMode=RCTM_Break, time=0.4169, value=1}
          {interpMode=RCIM_Cubic, leaveTangentWeight=0.2686, tangentMode=RCTM_User, time=0.8302, value=0}
         ]}
        GreenCurve {keys [
          {arriveTangentWeight=0.3287, interpMode=RCIM_Cubic, leaveTangentWeight=0.2272, tangentMode=RCTM_Break, time=0.4169, value=0.5536}
          {interpMode=RCIM_Cubic, leaveTangentWeight=0.2086, tangentMode=RCTM_User, time=0.8302, value=0}
         ]}
        RedCurve {keys [
          {arriveTangentWeight=0.1727, interpMode=RCIM_Cubic, leaveTangentWeight=0.1607, tangentMode=RCTM_Break, time=0.4169, value=0.056}
          {interpMode=RCIM_Cubic, leaveTangentWeight=0.1732, tangentMode=RCTM_User, time=0.8302, value=0}
         ]}
       }}
      CurveIndex=&ReturnNormalizedExecIndex()
     )
     Scale Factor=&MakeFloat4FromVectorAndFloat(
      "Vector (RGB)"=<1 1 1>
      "Float (Alpha)"=&FloatFromCurve(
       FloatCurve=DataInterface() {propertyValues {Curve {keys [
           {
            arriveTangent=-4.74
            arriveTangentWeight=0.6396
            interpMode=RCIM_Cubic
            leaveTangent=-4.74
            leaveTangentWeight=0.6396
            tangentMode=RCTM_User
            time=0
            value=2
           }
           {
            arriveTangent=0.0404
            arriveTangentWeight=0.614
            interpMode=RCIM_Cubic
            leaveTangent=0.0404
            leaveTangentWeight=0.614
            tangentMode=RCTM_User
            tangentWeightMode=RCTWM_WeightedBoth
            time=1
            value=0
           }
          ]}}}
       Scale Curve=0.5
      )
     )
    )) {enabled=false}
   &ScaleRibbonWidth(Ribbon Width Scale=&FloatFromCurve(FloatCurve=DataInterface() {propertyValues {Curve {keys [<at 0 1>,<at 1 0.9>]}}}))
   /SetVariables(Particles.Scaled Smoke Color:LinearColor=&MultiplyLinearColor(
     Linear Color=$Particles.Initial.Smoke Color
     Scale Factor=&MakeFloat4FromVectorAndFloat(
      "Vector (RGB)"=<1 1 1>
      "Float (Alpha)"=&Multiply_Float(
       A=&FloatFromCurve(
        FloatCurve=DataInterface() {propertyValues {Curve {keys [{tangentMode=RCTM_User, time=0, value=1}, {tangentMode=RCTM_User, time=1, value=0}]}}}
        Scale Curve=2
       )
       B=&FNTimeOfDayScalar(Night Scale=2.2, Morning Scale=1, Evening Scale=1)
      )
     )
    )) {isSetParameters=true, module=SetVariables_#4}
  ]
  rendererClasses [&NiagaraRibbonRendererProperties]
  renderers [
   {rendererClass=&NiagaraRibbonRendererProperties}
   {rendererClass=&NiagaraRibbonRendererProperties, rendererIndex=1}
  ]
 }
}
summary {
 systemName=NS_Dark_Pit_Pick_Swing
 emitters [
  {emitterName=FN_AnimTrailMain, rendererClasses [&NiagaraRibbonRendererProperties]}
  {bEnabled=false, emitterName=FN_AnimTrailMain001, rendererClasses [&NiagaraRibbonRendererProperties]}
  {bEnabled=false, emitterName=FN_AnimTrailMain002, rendererClasses [&NiagaraRibbonRendererProperties]}
 ]
 userVariables [
  User.RibbonIDOffset {type=NiagaraInt32}
  User.SkeletalMesh {
   type=NiagaraDataInterfaceSkeletalMesh
   defaultValue {
    struct=NiagaraExt_VariableValue_DataInterface
    value {dataInterface=&self:NiagaraDataInterfaceSkeletalMesh_0}
   }
  }
 ]
}
